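# Prototype master material for bug resin, kept in an S37 test map folder.
# It exposes Roughness, Base Color and SSS Color parameters, with both colors defaulting to the same muted green and no textures or math nodes.
# The shading model is Default Lit despite the subsurface color parameter.
# Nanite usage is enabled and automatic usage setting in the editor is off.
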Material /BRMapCh6/Maps/Test/ChrisS/S37_Proto/Materials/M_BugResin_Proto_MASTER
graph {
 properties {bAutomaticallySetUsageInEditor=false, bUsedWithNanite=true, usedShadingModels=Default Lit}
 nodes [
  ScalarParameter() {defaultValue=0.5, parameterName=Roughness}
  VectorParameter() {parameterName=Base Color, defaultValue=<rgba 0.0751 0.1406 0.0783 1>}
  VectorParameter() {parameterName="SSS  Color", defaultValue=<rgba 0.0751 0.1406 0.0783 1>}
 ]
}
